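# Sound cue for the tier 4 Bolt Revolver pistol's gunshot, built from borrowed shotgun and distant-shot samples.
# Each branch mixes a close shot (sharp crack, delayed strange-shot layer, delayed pitched-down tail) with a quieter distant shot.
# Close and distant layers carry their own attenuation and sound class overrides for third-person and distant fire.
# A quality switch fills two slots, the second dropping random tail and distant variants, and leaves two empty.
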SoundCue /Game/Sounds/Fort_Weapons/pistol/Weapon_PSTL_BoltRevolver_T04_Bolt_Revolver_Cue
root=QualityLevel_0
nodes [
 {id=Attenuation_2, children [Modulator_5], properties {attenuationSettings=/Game/Sounds/Attenuation/Fort_Weapon_Projectile_Distant_StW_Att}}
 {id=Attenuation_3, children [Mixer_3], properties {attenuationSettings=/Game/Sounds/Attenuation/Fort_Weapon_Pistol_Sml_Att}}
 {id=Attenuation_4, children [Modulator_8], properties {attenuationSettings=&Fort_Weapon_Projectile_Distant_StW_Att}}
 {id=Attenuation_5, children [Mixer_6], properties {attenuationSettings=&Fort_Weapon_Pistol_Sml_Att}}
 {id=Delay_2, children [Modulator_4], properties {delayMax=0.3, delayMin=0.3}}
 {id=Delay_3, children [WavePlayer_10], properties {delayMax=0.2, delayMin=0.2}}
 {id=Delay_4, children [Modulator_7], properties {delayMax=0.3, delayMin=0.3}}
 {id=Delay_5, children [WavePlayer_16], properties {delayMax=0.2, delayMin=0.2}}
 {id=Mixer_3, children [Modulator_3,Delay_2], properties {inputVolume [2,0.5]}}
 {id=Mixer_4, children [SoundClass_3,SoundClass_2], properties {inputVolume [1,1]}}
 {id=Mixer_5, children [WavePlayer_11, Delay_3], properties {inputVolume [0.6,1]}}
 {id=Mixer_6, children [Modulator_6,Delay_4], properties {inputVolume [2,0.5]}}
 {id=Mixer_7, children [SoundClass_5,SoundClass_4], properties {inputVolume [1,1]}}
 {id=Mixer_8, children [WavePlayer_17, Delay_5], properties {inputVolume [0.6,1]}}
 {id=Modulator_3, children [Mixer_5], properties {pitchMax=1, pitchMin=1}}
 {id=Modulator_4, children [Random_2], properties {pitchMax=0.6, pitchMin=0.6}}
 {id=Modulator_5, children [Random_3], properties {volumeMax=0.3, volumeMin=0.3}}
 {id=Modulator_6, children [Mixer_8], properties {pitchMax=1, pitchMin=1}}
 {id=Modulator_7, children [Random_4], properties {pitchMax=0.6, pitchMin=0.6}}
 {id=Modulator_8, children [WavePlayer_14], properties {volumeMax=0.3, volumeMin=0.3}}
 {id=QualityLevel_0, children [Mixer_4,Mixer_7,null,null]}
 {id=Random_2, children [WavePlayer_6,WavePlayer_7], properties {weights [1,1]}}
 {id=Random_3, children [WavePlayer_8,WavePlayer_9], properties {weights [1,1]}}
 {id=Random_4, children [WavePlayer_12], properties {weights [1]}}
 {
  id=SoundClass_2
  children [Attenuation_2]
  properties {soundClassOverride=&Fort_WeaponRanged_Fire_Distant}
 }
 {
  id=SoundClass_3
  children [Attenuation_3]
  properties {soundClassOverride=&Fort_WeaponRanged_Fire_3P_PSM}
 }
 {
  id=SoundClass_4
  children [Attenuation_4]
  properties {soundClassOverride=&Fort_WeaponRanged_Fire_Distant}
 }
 {
  id=SoundClass_5
  children [Attenuation_5]
  properties {soundClassOverride=&Fort_WeaponRanged_Fire_3P_PSM}
 }
 {id=WavePlayer_10, properties {soundWaveAssetPtr=/Game/Sounds/Fort_Weapons/Shotgun/shotgun_strangeshot_03}}
 {id=WavePlayer_11, properties {soundWaveAssetPtr=/Game/Sounds/Fort_Weapons/Shotgun/shotgun_sharpshot_04}}
 {id=WavePlayer_12, properties {soundWaveAssetPtr=/Game/Sounds/Fort_Weapons/Shotgun/Shotgun_Fire_Tail_Stereo_01}}
 {id=WavePlayer_14, properties {soundWaveAssetPtr=/Game/Sounds/Fort_Weapons/Distant/Distant_low_shot_01b}}
 {id=WavePlayer_16, properties {soundWaveAssetPtr=&shotgun_strangeshot_03}}
 {id=WavePlayer_17, properties {soundWaveAssetPtr=&shotgun_sharpshot_04}}
 {id=WavePlayer_6, properties {soundWaveAssetPtr=&Shotgun_Fire_Tail_Stereo_01}}
 {id=WavePlayer_7, properties {soundWaveAssetPtr=/Game/Sounds/Fort_Weapons/Shotgun/Shotgun_Fire_Tail_Stereo_02}}
 {id=WavePlayer_8, properties {soundWaveAssetPtr=&Distant_low_shot_01b}}
 {id=WavePlayer_9, properties {soundWaveAssetPtr=/Game/Sounds/Fort_Weapons/Distant/Distant_low_shot_01a}}
]
properties {
 bOverrideConcurrency=true
 maxDistance=20000
 volumeMultiplier=0.8
 attenuationOverrides {distanceAlgorithm=NaturalSound, falloffDistance=5000, attenuationShapeExtents {x=1000}}
}
