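# Sound cue for the start of the Cobra pistol's reload foley.
# A single wave player plays the machine pistol reload-start wave through a Modulator node at the root.
# It uses the shared weapon reload attenuation asset and weapon foley sound class, with slightly reduced volume.
# Concurrency is overridden to prevent new instances once the limit is reached.
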SoundCue /Game/Sounds/Fort_Weapons/pistol/Weapon_PSTL_Cobra_Reload_Start_Cue
root=Modulator_2
nodes [
 {id=Modulator_2, children [WavePlayer_27]}
 {id=WavePlayer_27, properties {soundWaveAssetPtr=/Game/Sounds/Fort_Weapons/pistol/pistol_foley/machine_pistol_reload_start}}
]
properties {
 attenuationSettings=&Fort_Weapon_Reload_Att
 bOverrideConcurrency=true
 maxDistance=2250
 soundClassObject=&Fort_WeaponFoley
 volumeMultiplier=0.802
 attenuationOverrides {falloffDistance=4100, attenuationShapeExtents {x=900}}
 concurrencyOverrides {resolutionRule=PreventNew}
}
