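# Sound cue for the looping fire audio of the Cyclone gatling pistol, using third-person and distant ranged-weapon fire sound classes.
# A root mixer blends three layers, each with its own sound class and attenuation.
# The close layer mixes a start one-shot with a firing loop; the distant layer reuses an auto-rifle loop pitched up by random modulation.
# A quieter shell-casing loop enters after a fixed half-second delay.
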SoundCue /Game/Sounds/Fort_Weapons/pistol/Weapon_PSTL_Gatling_Cyclone_Cue
root=Mixer_17
nodes [
 {id=Attenuation_0, children [Delay_0], properties {attenuationSettings=/Game/Sounds/Attenuation/Fort_Weapon_Rifle_Sml_Att}}
 {id=Attenuation_34, children [Modulator_12], properties {attenuationSettings=/Game/Sounds/Attenuation/Fort_Weapon_Projectile_Distant_StW_Att}}
 {id=Attenuation_49, children [Mixer_0], properties {attenuationSettings=&Fort_Weapon_Rifle_Sml_Att}}
 {id=Delay_0, children [WavePlayer_4], properties {delayMax=0.5, delayMin=0.5}}
 {id=Mixer_0, children [WavePlayer_0,WavePlayer_3], properties {inputVolume [1,0.6]}}
 {id=Mixer_17, children [SoundClass_32, SoundClass_35, SoundClass_0], properties {inputVolume [1,1,0.4]}}
 {id=Modulator_12, children [WavePlayer_67], properties {pitchMax=1.2, pitchMin=1.1}}
 {
  id=SoundClass_0
  children [Attenuation_0]
  properties {soundClassOverride=&Fort_WeaponRanged_Fire_3P_PSM}
 }
 {
  id=SoundClass_32
  children [Attenuation_49]
  properties {soundClassOverride=&Fort_WeaponRanged_Fire_3P_PSM}
 }
 {
  id=SoundClass_35
  children [Attenuation_34]
  properties {soundClassOverride=&Fort_WeaponRanged_Fire_Distant}
 }
 {id=WavePlayer_0, properties {bLooping=true, soundWaveAssetPtr=&/Weapon_Pistol_Gatling_01_Loop}}
 {id=WavePlayer_3, properties {soundWaveAssetPtr=&/Weapon_Pistol_Gatling_Start_01}}
 {id=WavePlayer_4, properties {bLooping=true, soundWaveAssetPtr=&/Weapon_PSTL_Shell_Casings_Loop}}
 {id=WavePlayer_67, properties {bLooping=true, soundWaveAssetPtr=/Game/Sounds/Fort_Weapons/AutoRifle/ar_pol_distant_loop_01}}
]
properties {bDebug=true, bOverrideConcurrency=true, maxDistance=20000, volumeMultiplier=0.95}
