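# Sound cue for the stop sound of the Cyclone gatling pistol weapon.
# A Mixer layers the gatling end wave with a quieter shell-casings end wave, each passing through its own default Modulator.
# It uses a long-range AR attenuation, a third-person ranged fire sound class, and a per-owner concurrency rule that stops the oldest instance.
# The cue's debug flag is left enabled.
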SoundCue /Game/Sounds/Fort_Weapons/pistol/Weapon_PSTL_Gatling_Cyclone_Stop_Cue
root=Mixer_0
nodes [
 {id=Mixer_0, children [Modulator_1,Modulator_0], properties {inputVolume [1,0.4]}}
 {id=Modulator_0, children [WavePlayer_1]}
 {id=Modulator_1, children [WavePlayer_0]}
 {id=WavePlayer_0, properties {soundWaveAssetPtr=&/Weapon_Pistol_Gatling_End_01}}
 {id=WavePlayer_1, properties {soundWaveAssetPtr=&/Weapon_PSTL_Shell_Casings_End}}
]
properties {
 attenuationSettings=/Game/Sounds/Attenuation/Fort_Weapon_AR_Lng_Att
 bDebug=true
 maxDistance=4000
 soundClassObject=&Fort_WeaponRanged_Fire_3P_PSM
 volumeMultiplier=1
 concurrencySet [/Game/Sounds/Concurrency_Rules/OneMaximum_Owner_StopOldest]
}
