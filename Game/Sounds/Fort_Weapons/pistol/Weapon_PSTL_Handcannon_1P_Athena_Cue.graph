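# Sound cue for first-person hand cannon pistol gunfire in Athena, using close-range fire recordings.
# A Random node picks between two equally weighted fire waves and is excluded from branch culling, with a Modulator above it at the root.
# The cue routes through a first-person ranged-fire sound class, first-person projectile attenuation and a first-person gunfire concurrency set.
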SoundCue /Game/Sounds/Fort_Weapons/pistol/Weapon_PSTL_Handcannon_1P_Athena_Cue
root=Modulator_2
nodes [
 {id=Modulator_2, children [Random_6]}
 {id=Random_6, children [WavePlayer_2,WavePlayer_3], properties {bShouldExcludeFromBranchCulling=true, weights [1,1]}}
 {id=WavePlayer_2, properties {soundWaveAssetPtr=/Game/Sounds/Fort_Weapons/pistol/HandCannon/Handcannon_Fire_Close_01}}
 {id=WavePlayer_3, properties {soundWaveAssetPtr=/Game/Sounds/Fort_Weapons/pistol/HandCannon/Handcannon_Fire_Close_02}}
]
properties {
 attenuationSettings=&Fort_Weapon_Projectile_FirstPerson_Athena_Att
 maxDistance=2097152
 soundClassObject=&Fort_WeaponRanged_Fire_1P_PSM
 volumeMultiplier=1
 attenuationOverrides {distanceAlgorithm=NaturalSound, falloffDistance=5000, attenuationShapeExtents {x=1000}}
 concurrencySet [&Weapon_Gunfire_FirstPerson_Conc]
}
