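# Sound cue for Handcannon pistol gunfire in Athena, built from close, medium and distant layers.
# Each layer randomly picks one of two waves, then passes through a modulator, its own distance attenuation and a sound class override.
# Close and medium layers share a third-person fire class and are mixed first, then joined by the much quieter distant layer with its own class.
# Local and global gunfire concurrency sets govern simultaneous playback.
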SoundCue /Game/Sounds/Fort_Weapons/pistol/Weapon_PSTL_Handcannon_Athena_Cue
root=Mixer_0
nodes [
 {id=Attenuation_0, children [Modulator_2], properties {attenuationSettings=&Fort_Weapon_Projectile_Close_Athena_Att}}
 {
  id=Attenuation_1
  children [Modulator_0]
  properties {attenuationSettings=&Fort_Weapon_Projectile_Distant_Mono_Athena_Att}
 }
 {
  id=Attenuation_6
  children [Modulator_1]
  properties {attenuationSettings=&Fort_Weapon_Projectile_Medium_Mono_Athena_Att}
 }
 {id=Mixer_0, children [Mixer_1,SoundClass_1], properties {inputVolume [1,0.085]}}
 {id=Mixer_1, children [SoundClass_0,SoundClass_6], properties {inputVolume [0.9,0.7]}}
 {id=Modulator_0, children [Random_3], properties {volumeMax=2, volumeMin=2}}
 {id=Modulator_1, children [Random_2], properties {volumeMax=2, volumeMin=2}}
 {id=Modulator_2, children [Random_6]}
 {id=Random_2, children [WavePlayer_0,WavePlayer_1], properties {bShouldExcludeFromBranchCulling=true, weights [1,1]}}
 {id=Random_3, children [WavePlayer_4,WavePlayer_5], properties {bShouldExcludeFromBranchCulling=true, weights [1,1]}}
 {id=Random_6, children [WavePlayer_2,WavePlayer_3], properties {bShouldExcludeFromBranchCulling=true, weights [1,1]}}
 {
  id=SoundClass_0
  children [Attenuation_0]
  properties {soundClassOverride=&Fort_WeaponRanged_Fire_3P_PSM}
 }
 {
  id=SoundClass_1
  children [Attenuation_1]
  properties {soundClassOverride=&Fort_WeaponRanged_Fire_Distant}
 }
 {
  id=SoundClass_6
  children [Attenuation_6]
  properties {soundClassOverride=&Fort_WeaponRanged_Fire_3P_PSM}
 }
 {id=WavePlayer_0, properties {soundWaveAssetPtr=/Game/Sounds/Fort_Weapons/pistol/HandCannon/Handcannon_Fire_MediumNew_01}}
 {id=WavePlayer_1, properties {soundWaveAssetPtr=/Game/Sounds/Fort_Weapons/pistol/HandCannon/Handcannon_Fire_MediumNew_02}}
 {id=WavePlayer_2, properties {soundWaveAssetPtr=/Game/Sounds/Fort_Weapons/pistol/HandCannon/Handcannon_Fire_Close_01}}
 {id=WavePlayer_3, properties {soundWaveAssetPtr=/Game/Sounds/Fort_Weapons/pistol/HandCannon/Handcannon_Fire_Close_02}}
 {id=WavePlayer_4, properties {soundWaveAssetPtr=/Game/Sounds/Fort_Weapons/pistol/HandCannon/Handcannon_Fire_DistantNew_01}}
 {id=WavePlayer_5, properties {soundWaveAssetPtr=/Game/Sounds/Fort_Weapons/pistol/HandCannon/Handcannon_Fire_DistantNew_02}}
]
properties {
 maxDistance=27500
 volumeMultiplier=0.65
 attenuationOverrides {distanceAlgorithm=NaturalSound, falloffDistance=5000, attenuationShapeExtents {x=1000}}
 concurrencySet [&Weapon_Gunfire_Local_Conc, &Weapon_Gunfire_Global_Conc]
}
