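# First-person fire sound cue for the Vigilante semi-auto pistol in Athena.
# A single pre-mixed fire wave plays through a Modulator into a Mixer, whose other two inputs are empty.
# It uses a first-person projectile attenuation asset, a ranged-weapon 1P fire sound class and the first-person gunfire concurrency set.
# Concurrency override values describe a small per-owner voice limit that stops the oldest.
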SoundCue /Game/Sounds/Fort_Weapons/pistol/Weapon_PSTL_SemiAuto_Fire_Vigilante_1P_Athena_Cue
root=Mixer_2
nodes [
 {id=Mixer_2, children [Modulator_6,null,null], properties {inputVolume [1,0.6,0.3]}}
 {id=Modulator_6, children [WavePlayer_0]}
 {id=WavePlayer_0, properties {soundWaveAssetPtr=&/Pistol_Vigilante_FireMix_01}}
]
properties {
 attenuationSettings=&Fort_Weapon_Projectile_FirstPerson_Athena_Att
 maxDistance=2097152
 soundClassObject=&Fort_WeaponRanged_Fire_1P_PSM
 volumeMultiplier=1
 attenuationOverrides {falloffDistance=4100, attenuationShapeExtents {x=900}}
 concurrencyOverrides {bLimitToOwner=true, maxCount=2, resolutionRule=StopOldest}
 concurrencySet [&Weapon_Gunfire_FirstPerson_Conc]
}
