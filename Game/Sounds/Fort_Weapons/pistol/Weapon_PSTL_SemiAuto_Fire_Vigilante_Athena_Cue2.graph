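# Sound cue for the Vigilante semi-auto pistol's gunshot in Athena, classed as third-person ranged weapon fire.
# A mixer layers close, medium and distant branches at decreasing input volume, each with its own attenuation asset and sound class override, the distant one using a separate class.
# The close layer plays one wave, while medium and distant pick randomly between two with boosted volume.
# Local and global gunfire concurrency sets are assigned.
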SoundCue /Game/Sounds/Fort_Weapons/pistol/Weapon_PSTL_SemiAuto_Fire_Vigilante_Athena_Cue2
root=Mixer_2
nodes [
 {id=Attenuation_3, children [Modulator_6], properties {attenuationSettings=&Fort_Weapon_Projectile_Close_Athena_Att}}
 {
  id=Attenuation_6
  children [Modulator_4]
  properties {attenuationSettings=&Fort_Weapon_Projectile_Medium_Mono_Athena_Att}
 }
 {
  id=Attenuation_7
  children [Modulator_5]
  properties {attenuationSettings=&Fort_Weapon_Projectile_Distant_Mono_Athena_Att}
 }
 {id=Mixer_2, children [SoundClass_3, SoundClass_10, SoundClass_9], properties {inputVolume [1,0.6,0.3]}}
 {id=Modulator_4, children [Random_1], properties {volumeMax=2, volumeMin=2}}
 {id=Modulator_5, children [Random_0], properties {volumeMax=2, volumeMin=2}}
 {id=Modulator_6, children [WavePlayer_0]}
 {id=Random_0, children [WavePlayer_1,WavePlayer_2], properties {weights [1,1]}}
 {id=Random_1, children [WavePlayer_3,WavePlayer_4], properties {weights [1,1]}}
 {
  id=SoundClass_10
  children [Attenuation_6]
  properties {soundClassOverride=&Fort_WeaponRanged_Fire_3P_PSM}
 }
 {
  id=SoundClass_3
  children [Attenuation_3]
  properties {soundClassOverride=&Fort_WeaponRanged_Fire_3P_PSM}
 }
 {
  id=SoundClass_9
  children [Attenuation_7]
  properties {soundClassOverride=&Fort_WeaponRanged_Fire_Distant}
 }
 {id=WavePlayer_0, properties {soundWaveAssetPtr=&/Pistol_Vigilante_FireMix_01}}
 {id=WavePlayer_1, properties {soundWaveAssetPtr=&/Pistol_Light_DistantShotNew_01}}
 {id=WavePlayer_2, properties {soundWaveAssetPtr=&/Pistol_Light_DistantShotNew_02}}
 {id=WavePlayer_3, properties {soundWaveAssetPtr=&/pistol_mediumFireNew_A_01}}
 {id=WavePlayer_4, properties {soundWaveAssetPtr=&/pistol_mediumFireNew_A_02}}
]
properties {
 maxDistance=27500
 soundClassObject=&Fort_WeaponRanged_Fire_3P_PSM
 volumeMultiplier=1
 attenuationOverrides {falloffDistance=4100, attenuationShapeExtents {x=900}}
 concurrencyOverrides {bLimitToOwner=true, maxCount=2, resolutionRule=StopOldest}
 concurrencySet [&Weapon_Gunfire_Local_Conc, &Weapon_Gunfire_Global_Conc]
}
