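# Sound cue for the start of a semi-auto pistol reload, playing a single magazine-eject foley wave.
# A Wave Player feeds a Modulator at the root, and the cue is assigned to the weapon foley sound class.
# It references the shared weapon reload attenuation and overrides concurrency so new instances are refused rather than replacing playing ones when the limit is reached.
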SoundCue /Game/Sounds/Fort_Weapons/pistol/Weapon_PSTL_SemiAuto_Reload_Start_Cue
root=Modulator_2
nodes [
 {id=Modulator_2, children [WavePlayer_12]}
 {id=WavePlayer_12, properties {soundWaveAssetPtr=/Game/Sounds/Fort_Weapons/pistol/pistol_foley/auto_sdk_eject_01}}
]
properties {
 attenuationSettings=&Fort_Weapon_Reload_Att
 bOverrideConcurrency=true
 maxDistance=2250
 soundClassObject=&Fort_WeaponFoley
 volumeMultiplier=0.752
 attenuationOverrides {falloffDistance=4100, attenuationShapeExtents {x=900}}
 concurrencyOverrides {resolutionRule=PreventNew}
}
