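# Fire sound cue for the Vigilante basic semi-auto T01 pistol, with owner-based projectile concurrency.
# Each branch mixes a near layer (random .45 shot, boosted tail, pitched-up sweetener) with a quieter distant shot, each having its own attenuation and sound class.
# A quality switch selects between two near-identical branches, the second using two random shot variations instead of four.
# The remaining two quality slots are empty.
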SoundCue /Game/Sounds/Fort_Weapons/pistol/Weapon_PSTL_SemiAuto_T01_Basic_Vigilante_Cue
root=QualityLevel_0
nodes [
 {id=Attenuation_2, children [Modulator_5], properties {attenuationSettings=/Game/Sounds/Attenuation/Fort_Weapon_Projectile_Distant_StW_Att}}
 {id=Attenuation_3, children [Modulator_6], properties {attenuationSettings=/Game/Sounds/Attenuation/Fort_Weapon_Pistol_Sml_Att}}
 {id=Attenuation_4, children [Modulator_9], properties {attenuationSettings=&Fort_Weapon_Projectile_Distant_StW_Att}}
 {id=Attenuation_5, children [Modulator_10], properties {attenuationSettings=&Fort_Weapon_Pistol_Sml_Att}}
 {id=Mixer_2, children [SoundClass_3,SoundClass_2], properties {inputVolume [1,0.8]}}
 {id=Mixer_3, children [Modulator_4, WavePlayer_13, Modulator_7], properties {inputVolume [1,1.6,1]}}
 {id=Mixer_4, children [SoundClass_5,SoundClass_4], properties {inputVolume [1,0.8]}}
 {id=Mixer_5, children [Modulator_8, WavePlayer_20, Modulator_11], properties {inputVolume [1,1.6,1]}}
 {id=Modulator_10, children [Mixer_5]}
 {id=Modulator_11, children [WavePlayer_19], properties {pitchMax=2, pitchMin=2, volumeMax=1.2, volumeMin=1.2}}
 {id=Modulator_4, children [Random_1]}
 {id=Modulator_5, children [WavePlayer_11], properties {volumeMax=0.3, volumeMin=0.3}}
 {id=Modulator_6, children [Mixer_3]}
 {id=Modulator_7, children [WavePlayer_12], properties {pitchMax=2, pitchMin=2, volumeMax=1.2, volumeMin=1.2}}
 {id=Modulator_8, children [Random_2]}
 {id=Modulator_9, children [WavePlayer_18], properties {volumeMax=0.3, volumeMin=0.3}}
 {id=QualityLevel_0, children [Mixer_2,Mixer_4,null,null]}
 {id=Random_1, children [WavePlayer_7, WavePlayer_8, WavePlayer_9, WavePlayer_10], properties {weights [1,1,1,1]}}
 {id=Random_2, children [WavePlayer_14, WavePlayer_16], properties {weights [1,1]}}
 {
  id=SoundClass_2
  children [Attenuation_2]
  properties {soundClassOverride=&Fort_WeaponRanged_Fire_Distant}
 }
 {
  id=SoundClass_3
  children [Attenuation_3]
  properties {soundClassOverride=&Fort_WeaponRanged_Fire_3P_PSM}
 }
 {
  id=SoundClass_4
  children [Attenuation_4]
  properties {soundClassOverride=&Fort_WeaponRanged_Fire_Distant}
 }
 {
  id=SoundClass_5
  children [Attenuation_5]
  properties {soundClassOverride=&Fort_WeaponRanged_Fire_3P_PSM}
 }
 {id=WavePlayer_10, properties {soundWaveAssetPtr=&/Pistol_45_Fire_Stereo_04}}
 {id=WavePlayer_11, properties {soundWaveAssetPtr=/Game/Sounds/Fort_Weapons/Distant/Distant_high_shot_01b}}
 {id=WavePlayer_12, properties {soundWaveAssetPtr=/Game/Sounds/Fort_Weapons/WeaponSweeteners/weapon_sweet_08}}
 {id=WavePlayer_13, properties {soundWaveAssetPtr=&/tail_shot_03}}
 {id=WavePlayer_14, properties {soundWaveAssetPtr=&/Pistol_45_Fire_Stereo_01}}
 {id=WavePlayer_16, properties {soundWaveAssetPtr=&/Pistol_45_Fire_Stereo_03}}
 {id=WavePlayer_18, properties {soundWaveAssetPtr=&Distant_high_shot_01b}}
 {id=WavePlayer_19, properties {soundWaveAssetPtr=&weapon_sweet_08}}
 {id=WavePlayer_20, properties {soundWaveAssetPtr=&tail_shot_03}}
 {id=WavePlayer_7, properties {soundWaveAssetPtr=&Pistol_45_Fire_Stereo_01}}
 {id=WavePlayer_8, properties {soundWaveAssetPtr=&/Pistol_45_Fire_Stereo_02}}
 {id=WavePlayer_9, properties {soundWaveAssetPtr=&Pistol_45_Fire_Stereo_03}}
]
properties {
 maxDistance=20000
 soundClassObject=&Fort_WeaponRanged_Fire_3P_PSM
 volumeMultiplier=0.4
 attenuationOverrides {falloffDistance=4100, attenuationShapeExtents {x=900}}
 concurrencySet [/Game/Sounds/Concurrency_Rules/Weapon_Projectile_Owner_Conc]
}
